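; Storage box cluster prop Blueprint for Apollo environments, with a construction script that can randomly thin out its oil drum components.
; When RandomComponentRemoval is true, the script runs the parent construction script, then destroys either S_Constr_OilDrum_02 or S_Constr_OilDrum_03 at random.
; It then clears that flag within the same branch.
; The BeginPlay, Tick and overlap events are empty.
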(event EventBeginPlay)

(event Collision|EventActorBeginOverlap
  (params (OtherActor :type Actor%20Object%20Reference)))

(event EventTick
  (params (DeltaSeconds :type Float%20%28single-precision%29)))

(fn ConstructionScript
  (|CallParentFunction)
  (if (Variables|Default|GetRandomComponentRemoval)
    (Components|DestroyComponent (Utilities|Array|Get(acopy) (Utilities|Array|MakeArray (Variables|Default|GetS_Constr_OilDrum_03) (Variables|Default|GetS_Constr_OilDrum_02)) (Math|Random|RandomInteger (- (Utilities|Array|Length (Utilities|Array|MakeArray (Variables|Default|GetS_Constr_OilDrum_03) (Variables|Default|GetS_Constr_OilDrum_02))) 0))))
    (Variables|Default|SetRandomComponentRemoval false)))
